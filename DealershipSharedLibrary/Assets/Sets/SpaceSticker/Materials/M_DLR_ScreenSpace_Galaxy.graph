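# Translucent unlit Niagara and particle material that renders a scrolling galaxy effect, from the SpaceSticker set.
# Screen-space starfield background and star-noise foreground are both modulated by a UV-panned noise middleground, which also adds its own tint.
# Their sum, multiplied by particle color, feeds emissive adjustments.
# An optional alpha mask, distorted by two panning noise samples, multiplies particle alpha before depth fade.
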Material /DealershipSharedLibrary/Assets/Sets/SpaceSticker/Materials/M_DLR_ScreenSpace_Galaxy
graph {
 properties {
  bAutomaticallySetUsageInEditor=false
  bUsedWithMeshParticles=true
  bUsedWithNiagaraMeshParticles=true
  bUsedWithNiagaraRibbons=true
  bUsedWithNiagaraSprites=true
  blendMode=BLEND_Translucent
  shadingModel=MSM_Unlit
  usedShadingModels=Unlit
 }
 nodes [
  Add(A=Multiply(
    A=VectorParameter() {group=Color, parameterName=Main Color, sortPriority=0, defaultValue=<rgba 0.1389 0.0301 0.083 1>}.RGB
    B=Constant3Vector()
   ))
  n1 = ParticleColor()
  n2 = NamedRerouteUsage()
  MaterialFunctionCall(Opacity In=Saturate(None=Multiply(A=@n1.A, B=@n2))) {materialFunction=&MF_VCVFX_DepthFade}
  n3 = VectorParameter() {group=01_Background, parameterName=Color Background, sortPriority=1, defaultValue=<rgba 0.3681 0.0153 0.0923 1>}
  n4 = ScalarParameter() {defaultValue=2, group=01_Background, parameterName=Strength Background, sortPriority=3}
  n5 = ScreenPosition()
  n6 = ComponentMask(None=@n5.ViewportUV) {g=true, r=true}
  n7 = Panner(Speed=AppendVector(
    A=ScalarParameter() {defaultValue=0.1, group=01_Background, parameterName=Scroll U Background, sortPriority=1}
    B=ScalarParameter() {defaultValue=0.1, group=01_Background, parameterName=Scroll V Background, sortPriority=1}
   ))
  n8 = ScalarParameter() {defaultValue=1, group=01_Background, parameterName=Tile U Background, sortPriority=2}
  n9 = ScalarParameter() {defaultValue=1, group=01_Background, parameterName=Tile V Background, sortPriority=2}
  n10 = TextureSampleParameter2D(UVs=Multiply(A=Add(A=@n6, B=@n7), B=AppendVector(A=@n8, B=@n9))) {
   group=01_Background
   parameterName="Texture Background (Screenspace)"
   sortPriority=0
   texture=/DealershipSharedLibrary/Assets/Sets/SpaceSticker/Textures/T_Starfield
  }
  n11 = TextureCoordinate()
  n12 = AppendVector(
   A=ScalarParameter() {defaultValue=-0.3, group=02_Middleground, parameterName=Scroll U Middleground, sortPriority=1}
   B=ScalarParameter() {defaultValue=-0.1, group=02_Middleground, parameterName=Scroll V Middleground, sortPriority=1}
  )
  n13 = ScalarParameter() {defaultValue=1, group=02_Middleground, parameterName=Tile U Middleground, sortPriority=2}
  n14 = ScalarParameter() {defaultValue=1, group=02_Middleground, parameterName=Tile V Middleground, sortPriority=2}
  n15 = TextureSampleParameter2D(UVs=Multiply(A=Panner(Coordinate=@n11, Speed=@n12), B=AppendVector(A=@n13, B=@n14))) {
   group=02_Middleground
   parameterName=Texture Middleground
   sortPriority=0
   texture=/Game/Effects/Fort_Effects/Textures/Tile/T_GaussianNoise_01_M
  }
  n16 = VectorParameter() {group=02_Middleground, parameterName=Color Middleground, sortPriority=1, defaultValue=<rgba 0 0.0134 0.0503 1>}
  n17 = Add(
   A=Multiply(A=Multiply(A=@n3.RGB, B=@n4) {constB=20}, B=Multiply(A=@n10.R, B=@n15.R))
   B=Multiply(
    A=Multiply(A=@n15.R, B=@n16.RGB)
    B=ScalarParameter() {defaultValue=0.1, group=02_Middleground, parameterName=Strength Middleground, sortPriority=3}
   ) {constB=20}
  )
  n18 = ComponentMask(None=@n5.ViewportUV) {g=true, r=true}
  n19 = Panner(Speed=AppendVector(
    A=ScalarParameter() {defaultValue=-0.1, group=03_Foreground, parameterName=Scroll U Foreground, sortPriority=1}
    B=ScalarParameter() {defaultValue=0.1, group=03_Foreground, parameterName=Scroll V Foreground, sortPriority=1}
   ))
  n20 = ScalarParameter() {defaultValue=1, group=03_Foreground, parameterName=Tile U Foreground, sortPriority=2}
  n21 = ScalarParameter() {defaultValue=1, group=03_Foreground, parameterName=Tile V Foreground, sortPriority=2}
  n22 = TextureSampleParameter2D(UVs=Multiply(A=Add(A=@n18, B=@n19), B=AppendVector(A=@n20, B=@n21))) {
   group=03_Foreground
   parameterName="Texture Foreground (Screenspace)"
   sortPriority=0
   texture=/Game/Effects/Fort_Effects/Gameplay/Butterfly/T_noise_stars
  }
  n23 = VectorParameter() {group=03_Foreground, parameterName=Color Foreground, sortPriority=1, defaultValue=<rgba 1 0.6988 0.2188 1>}
  n24 = Multiply(
   A=@n15.R
   B=Multiply(
    A=Multiply(A=@n22.RGB, B=@n23.RGB)
    B=ScalarParameter() {defaultValue=2, group=03_Foreground, parameterName=Strength Foreground, sortPriority=3}
   ) {constB=20}
  )
  MaterialFunctionCall(Emissive=Multiply(A=@n1.RGB, B=Add(A=@n17, B=@n24))) {materialFunction=&MF_FX_EmissiveAdjustments}
  n25 = TextureCoordinate()
  n26 = AppendVector(
   A=ScalarParameter() {defaultValue=-0.5, group=Distortion OpacityMask, parameterName=Scroll U Alpha Distortion, sortPriority=1}
   B=ScalarParameter() {defaultValue=-0.4, group=Distortion OpacityMask, parameterName=Scroll V Alpha Distortion, sortPriority=1}
  )
  n27 = TextureSampleParameter2D(UVs=Multiply(
    A=Panner(Coordinate=@n25, Speed=@n26)
    B=ScalarParameter() {defaultValue=1, group=Distortion OpacityMask, parameterName=Tile Alpha Distortion, sortPriority=2}
   )) {
   group=Distortion OpacityMask
   parameterName=AlphaDistort Texture
   sortPriority=0
   texture=/DealershipSharedLibrary/FX/Textures/Noise/T_VCVFX_Alpine_Infinite_RGBNoise01
  }
  n28 = AppendVector(
   A=ScalarParameter() {defaultValue=0.5, group=Distortion OpacityMask 2, parameterName=Scroll U Alpha Distortion 2, sortPriority=1}
   B=ScalarParameter() {defaultValue=0.4, group=Distortion OpacityMask 2, parameterName=Scroll V Alpha Distortion 2, sortPriority=1}
  )
  n29 = TextureSampleParameter2D(UVs=Multiply(
    A=Panner(Coordinate=@n25, Speed=@n28)
    B=ScalarParameter() {defaultValue=1, group=Distortion OpacityMask 2, parameterName=Tile Alpha Distortion 2, sortPriority=2}
   )) {
   group=Distortion OpacityMask 2
   parameterName=AlphaDistort Texture 2
   sortPriority=0
   texture=&T_VCVFX_Alpine_Infinite_RGBNoise01
  }
  n30 = Multiply(
   A=StaticComponentMaskParameter(None=@n27.RGB) {defaultR=true, group=Distortion OpacityMask, parameterName=Alpha Distorion Texture Channel}
   B=StaticComponentMaskParameter(None=@n29.RGB) {defaultG=true, group=Distortion OpacityMask 2, parameterName=Alpha Distorion Texture Channel 2}
  )
  n31 = ScalarParameter() {defaultValue=0.2688, group=Distortion OpacityMask Global, parameterName=AlphaDistortion Strength}
  n32 = TextureSampleParameter2D(UVs=Add(A=@n25, B=Multiply(A=@n30, B=@n31))) {
   group=Distortion OpacityMask Global
   parameterName=AlphaMask Texture
   sortPriority=0
   texture=/DealershipSharedLibrary/Assets/Sets/Infinite/T_Infinite_MaskBack_A
  }
  n33 = Power(Base=@n32.RGB, Exp=ScalarParameter() {defaultValue=1, group=Distortion OpacityMask Global, parameterName=Alpha Mask Size, sortPriority=3})
  n34 = ScalarParameter() {defaultValue=1, group=Distortion OpacityMask Global, parameterName=Alpha Mask Softness, sortPriority=4}
  n35 = ScalarParameter() {
   controlType=Enumeration
   defaultValue=1
   enumeration=/DealershipSharedLibrary/FX/Materials/Enums/Enum_DLR_TrueFalse
   group=Distortion OpacityMask Global
   parameterName=UseAlphaMask?
   sortPriority=0
  }
  NamedRerouteDeclaration(None=LinearInterpolate(B=Multiply(A=@n33, B=@n34), Alpha=Saturate(None=@n35)) {constA=1}) {name=OpacityMask, nodeColor=<rgba 0 1 0.4844 1>}
 ]
}
